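; BP_S22_WW_Avian_Q05_ChickPen is a chicken pen quest actor for the Season 22 Wild Weeks Avian quest 05, detecting chickens thrown inside.
; When its Chicken_Checker component begins an overlap, it casts the other actor to B_Prj_NugGuide.
; A valid cast resolves the instigator's player controller and sends the Athena.Quests.Irwin.Nug.ReturnedToPen stat event to that player's quest manager.
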(event EventBeginPlay
  (|CallParentFunction))

(event Collision|EventActorBeginOverlap
  (params (OtherActor :type Actor%20Object%20Reference))
  (|CallParentFunction OtherActor) ; This node is disabled and will not be called.\nDrag off pins to build functionality.
)

(event EventTick
  (params (DeltaSeconds :type Float%20%28single-precision%29))
  (|CallParentFunction DeltaSeconds) ; This node is disabled and will not be called.\nDrag off pins to build functionality.
)

(event OnComponentBeginOverlap(Chicken_Checker)
  (params (OverlappedComponent :type Primitive%20Component%20Object%20Reference) (OtherActor :type Actor%20Object%20Reference) (OtherComp :type Primitive%20Component%20Object%20Reference) (OtherBodyIndex :type Integer) (bFromSweep :type Boolean) (SweepResult :type Hit%20Result%20Structure%20%28by%20ref%29))
  ;[1 1 1] Checking Pen for chicken thrown inside.
  (bind _asb_prj_nug_guide (Utilities|Casting|CastToB_Prj_NugGuide OtherActor))
  (if (Utilities|IsValid _asb_prj_nug_guide :!class /Script/Engine.KismetSystemLibrary)
    (Quest|SendComplexCustomStatEvent (Quests|GetQuestManager (Fortnite|GetFortPlayerControllerfromActor (Game|GetInstigator _asb_prj_nug_guide)) "Invalid") 0 "(GameplayTags=((TagName=\"Athena.Quests.Irwin.Nug.ReturnedToPen\")))" "(GameplayTags=)" 1)))

(fn ConstructionScript
  (|CallParentFunction))
